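# Sound cue for the launch sound of the Chrome directional launcher, a launch pad in Athena.
# A Random node picks among four equally weighted launch wave variations, and a Modulator adds a slight upward pitch variation on each play.
# It uses the trap sound class and trap attenuation, and its own concurrency override caps simultaneous instances at two.
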SoundCue /Game/Athena/Sounds/LaunchPad/DirectionalLauncher/Chrome_DirectionalLauncher_Launch_Cue
root=Modulator_1
nodes [
 {id=Modulator_1, children [Random_0], properties {pitchMax=1.03, pitchMin=1}}
 {id=Random_0, children [WavePlayer_1,WavePlayer_2,WavePlayer_3,WavePlayer_4], properties {weights [1,1,1,1]}}
 {id=WavePlayer_1, properties {soundWaveAssetPtr=&/Chrome_DirectionalLauncher_Launch_01}}
 {id=WavePlayer_2, properties {soundWaveAssetPtr=&/Chrome_DirectionalLauncher_Launch_02}}
 {id=WavePlayer_3, properties {soundWaveAssetPtr=&/Chrome_DirectionalLauncher_Launch_03}}
 {id=WavePlayer_4, properties {soundWaveAssetPtr=&/Chrome_DirectionalLauncher_Launch_04}}
]
properties {
 attenuationSettings=&Fort_Traps_Att
 bOverrideConcurrency=true
 maxDistance=6000
 soundClassObject=&Fort_Traps
 volumeMultiplier=1
 concurrencyOverrides {maxCount=2}
}
