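# Sound cue for the looping idle ambience of the Directional Launcher launch pad.
# A single looping wave player for the idle loop wave feeds a modulator at the root, which applies pitch and volume modulation at its default ranges.
# The cue uses a dedicated ambient attenuation asset and the Creative gadgets sound class, with reduced volume.
# Its concurrency override prevents new instances from starting once the limit is reached.
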SoundCue /Game/Athena/Sounds/LaunchPad/DirectionalLauncher/DirectionalLauncher_Ambient_Loop_Cue
root=Modulator_0
nodes [
 {id=Modulator_0, children [WavePlayer_1]}
 {
  id=WavePlayer_1
  properties {bLooping=true, soundWaveAssetPtr=&/DirectionalLauncher22_Idle_Loop}
 }
]
properties {
 attenuationSettings=&/DirectionalLauncher_Ambient_Att
 bOverrideConcurrency=true
 maxDistance=2700
 soundClassObject=&Fort_Gadgets_Creative
 volumeMultiplier=0.7
 concurrencyOverrides {resolutionRule=PreventNew}
}
