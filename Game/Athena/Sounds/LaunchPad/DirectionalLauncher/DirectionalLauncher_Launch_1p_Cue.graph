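# Sound cue for the first-person launch sound of the directional launcher launch pad, played without distance attenuation.
# A single wave player feeds a modulator that adds slight random pitch variation.
# Overridden concurrency allows two simultaneous instances and stops the oldest.
# Despite the 1p name, the cue is assigned to a third-person ranged weapon fire sound class.
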SoundCue /Game/Athena/Sounds/LaunchPad/DirectionalLauncher/DirectionalLauncher_Launch_1p_Cue
root=Modulator_0
nodes [
 {id=Modulator_0, children [WavePlayer_1], properties {pitchMax=1.03, pitchMin=0.98}}
 {id=WavePlayer_1, properties {soundWaveAssetPtr=&/DirectionalLauncher22_Launch}}
]
properties {
 bOverrideConcurrency=true
 maxDistance=2097152
 soundClassObject=&Fort_WeaponRanged_Fire_3P_PSM
 volumeMultiplier=1
 attenuationOverrides {bAttenuate=false}
 concurrencyOverrides {maxCount=2, resolutionRule=StopOldest}
}
