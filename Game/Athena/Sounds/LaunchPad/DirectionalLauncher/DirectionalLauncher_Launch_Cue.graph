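# Sound cue for the launch sound of the Athena Directional Launcher launch pad.
# A single launch wave plays through a modulator that randomizes pitch slightly on each play for variation.
# It uses the trap sound class and trap attenuation at reduced volume, with a concurrency override that limits how many instances play at once.
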SoundCue /Game/Athena/Sounds/LaunchPad/DirectionalLauncher/DirectionalLauncher_Launch_Cue
root=Modulator_0
nodes [
 {id=Modulator_0, children [WavePlayer_1], properties {pitchMax=1.03, pitchMin=0.98}}
 {id=WavePlayer_1, properties {soundWaveAssetPtr=&/DirectionalLauncher22_Launch}}
]
properties {
 attenuationSettings=&Fort_Traps_Att
 bOverrideConcurrency=true
 maxDistance=6000
 soundClassObject=&Fort_Traps
 volumeMultiplier=0.6
 concurrencyOverrides {maxCount=2}
}
